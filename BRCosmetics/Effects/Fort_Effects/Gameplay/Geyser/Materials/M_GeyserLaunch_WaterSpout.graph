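# Unlit translucent Niagara mesh-particle material for the water spout of a geyser launch effect.
# Panning noise distorts UVs for two gradient layers, offset by dynamic parameters, whose contrasted sum blends trailing and leading colors.
# Opacity combines that shape with an optional mask and a noise erosion driven by particle alpha.
# A texture panned over particle lifetime displaces vertices along normals, scaled by a dynamic parameter.
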Material /BRCosmetics/Effects/Fort_Effects/Gameplay/Geyser/Materials/M_GeyserLaunch_WaterSpout
graph {
 properties {
  bUsedWithNiagaraMeshParticles=true
  blendMode=BLEND_Translucent
  pixelDepthOffsetMode=PDOM_Legacy
  shadingModel=MSM_Unlit
  usedShadingModels=Unlit
 }
 nodes [
  n1 = VectorParameter() {group=Noise, parameterName=Noise_Transform, channelNames {a=SpeedY, b=SpeedX, g=ScaleY, r=ScaleX}, defaultValue=<rgba 1 1 0 1>}
  n2 = Multiply(A=TextureCoordinate(), B=AppendVector(A=@n1.ScaleX, B=@n1.ScaleX))
  n3 = AppendVector(A=@n1.SpeedX, B=@n1.SpeedX)
  n4 = NamedRerouteDeclaration(None=TextureSampleParameter2D(UVs=Panner(Coordinate=@n2, Speed=@n3)) {group=Noise, parameterName=NoiseTexture, texture=&T_Inky_Smoke_Tile}.R) {name=NoiseTex, nodeColor=<rgba 0 0.0859 1 1>}
  n5 = VectorParameter() {group=Noise, parameterName=NoiseDistortion, channelNames {a="", b="", g=NoiseStrY, r=NoiseStrX}, defaultValue=<rgba 0 0 0 1>}
  n6 = AppendVector(A=@n5.NoiseStrX, B=@n5.NoiseStrX)
  n7 = Multiply(
   A=Add(A=Multiply(A=@n4, B=@n6), B=TextureCoordinate() {vTiling=3})
   B=AppendVector(A=ScalarParameter() {defaultValue=1, parameterName=Tile_X}, B=ScalarParameter() {parameterName=Tile_Y})
  )
  n8 = AppendVector(
   A=ComponentMask(None=@n7) {r=true}
   B=Power(Base=ComponentMask(None=@n7) {g=true}, Exp=ScalarParameter() {parameterName=UV_BunchUp}) {constExponent=0.5}
  )
  n9 = DynamicParameter() {defaultValue=<rgba 1 1 1 40>, paramNames [Y_Offset, Layer2_Y_Offset, OpacityMultiplier, WPO_Str]}
  n10 = Add(A=@n8, B=AppendVector(A=Constant(), B=@n9.Y_Offset))
  n11 = VectorParameter() {parameterName=Layer2_Transform, channelNames {a=OffsetY, b=OffsetX, g=ScaleY, r=ScaleX}, defaultValue=<rgba 0 0 0 1>}
  n12 = Add(
   A=Multiply(A=AppendVector(A=@n11.ScaleX, B=@n11.ScaleX), B=@n8)
   B=AppendVector(A=@n11.OffsetX, B=Add(A=@n11.OffsetY, B=@n9.Layer2_Y_Offset))
  )
  n13 = MaterialFunctionCall(
   In=Add(
    A=TextureSample(UVs=@n10) {texture=/BRCosmetics/Weapons/FORT_Melee/Pickaxe_Cube_Ninja_Male/FX/T_Offset_Gradient_01}.B
    B=TextureSample(UVs=@n12) {texture=&T_Offset_Gradient_01}.B
   )
   Contrast=ScalarParameter() {group=Alpha, parameterName=AlphaContrast}
  ) {materialFunction=&CheapContrast}
  MaterialFunctionCall(Emissive=LinearInterpolate(
    A=VectorParameter() {parameterName=TrailingColor, defaultValue=<rgba 0 0 0 1>}.RGB
    B=VectorParameter() {parameterName=LeadingColor, defaultValue=<rgba 0 0 0 1>}.RGB
    Alpha=@n13.Result
   )) {materialFunction=&MF_FX_EmissiveAdjustments}
  n14 = ParticleColor()
  n15 = Multiply(
   A=@n13.Result
   B=DynamicParameter() {defaultValue=<rgba 1 1 1 40>, paramNames [Y_Offset, Layer2_Y_Offset, OpacityMultiplier, WPO_Str]}.OpacityMultiplier
  )
  n16 = TextureCoordinate()
  n17 = VectorParameter() {group=Mask, parameterName=MaskTex_Transform, channelNames {a=OffsetY, b=OffsetX, g=ScaleY, r=ScaleX}, defaultValue=<rgba 1 1 0 0>}
  n18 = AppendVector(A=@n17.ScaleX, B=@n17.ScaleX)
  n19 = TextureSampleParameter2D(UVs=Add(A=Multiply(A=@n16, B=@n18), B=AppendVector(A=@n17.OffsetX, B=@n17.OffsetX))) {
   group=Alpha
   parameterName=Mask
   samplerSource=SSM_Clamp_WorldGroupSettings
   texture=/Game/Effects/Fort_Effects/Textures/Masks/T_Square_Gradient_M
  }
  n20 = ScalarParameter() {group=Mask, parameterName=UseMaskTex}
  Multiply(
   A=SmoothStep(Min=Add(A=@n14.A) {constB=0.1}, Max=@n14.A, Value=NamedRerouteUsage())
   B=Multiply(A=Saturate(None=@n15), B=LinearInterpolate(B=@n19.R, Alpha=@n20) {constA=1})
  )
  n21 = TextureCoordinate()
  n22 = VectorParameter() {group=WPO, parameterName=WPO_Transform, channelNames {a=SpeedY, b=SpeedX, g=ScaleY, r=ScaleX}, defaultValue=<rgba 0 0 0 1>}
  n23 = AppendVector(A=@n22.ScaleX, B=@n22.ScaleX)
  n24 = TextureSampleParameter2D(UVs=Panner(Coordinate=Multiply(A=@n21, B=@n23), Time=ParticleRelativeTime(), Speed=AppendVector(A=@n22.SpeedX, B=@n22.SpeedX))) {
   group=WPO
   parameterName=WPO_OffsetTex
   samplerSource=SSM_Wrap_WorldGroupSettings
   texture=&T_Inky_Smoke_Tile
  }
  Multiply(
   A=Multiply(A=VertexNormalWS(), B=@n24.R)
   B=Multiply(
    A=DynamicParameter() {defaultValue=<rgba 1 1 1 40>, paramNames [Y_Offset, Layer2_Y_Offset, OpacityMultiplier, WPO_Str]}.WPO_Str
    B=ScalarParameter() {group=WPO, parameterName=WPO_BaseStr}
   )
  )
 ]
}
